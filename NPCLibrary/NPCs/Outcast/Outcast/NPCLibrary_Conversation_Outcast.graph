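# Conversation database for the Outcast NPC, defining its dialogue under the Conversation.Entry.Outcast tag.
# The entry point leads to one Speech node with a default greeting and a disapproving contextual line for players wearing one of five TeriyakiFish-family outfits.
# That line's priority comes from the OutfitReaction curve row.
# The speech then branches through link nodes to the shared NPCLibrary Quests and BasicServices conversations.
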ConversationDatabase /NPCLibrary/NPCs/Outcast/Outcast/NPCLibrary_Conversation_Outcast
entry_points [{tag=Conversation.Entry.Outcast, destinations [#1]}]
nodes [
 ConversationEntryPointNode_7 {
  id=#1
  class=ConversationEntryPointNode
  outputs [#2]
  properties {entryTag {tagName=Conversation.Entry.Outcast}}
 }
 Speech_C_4 {
  id=#2
  class=Speech_C
  outputs [#3, #4]
  properties {
   nodeName=""
   generalConfig {
    defaultMessage=I can tell you're gonna make a big splash..
    contextualMessages [{
      message="I honestly don't like what's going on here. "
      requirementMatchPolicy=RequireAny
      contextRequirements [{requirement {
         class=FortControllerRequirement_HasCID
         properties {allowedCIds [
           &CID_315_Athena_Commando_M_TeriyakiFish
           /BRCosmetics/Athena/Items/Cosmetics/Characters/CID_624_Athena_Commando_M_TeriyakiWarrior
           /BRCosmetics/Athena/Items/Cosmetics/Characters/CID_656_Athena_Commando_M_TeriyakiFishFreezerBurn
           /BRCosmetics/Athena/Items/Cosmetics/Characters/CID_722_Athena_Commando_M_TeriyakiFishAssassin
           /BRCosmetics/Athena/Items/Cosmetics/Characters/CID_813_Athena_Commando_M_TeriyakiAtlantis
          ]}
        }}]
      priority {curve {rowName=Tandem.Speech.Priority.OutfitReaction}}
     }]
   }
  }
 }
 ConversationLinkNode_0 {id=#4, class=ConversationLinkNode, properties {remoteEntryTag {tagName=Conversation.Entry.BasicServices}}}
 ConversationLinkNode_1 {id=#3, class=ConversationLinkNode, properties {remoteEntryTag {tagName=Conversation.Entry.NPCLibrary.Quests}}}
]
